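# Sound cue for the Athena shooting gallery highscore sound, filed under quest sounds.
# A single wave player feeds the root directly, with no randomization or modulation.
# The cue uses gadget-activate attenuation, a mission in-game sound class, a raised volume multiplier and a concurrency set borrowed from rifle bullet impacts.
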SoundCue /FigmentEnvironment/Athena/Sounds/Quests/Athena_ShootingGallery_Highscore_Cue
root=WavePlayer_0
nodes [{id=WavePlayer_0, properties {soundWaveAssetPtr=&/Athena_ShootingGallery_Highscore}}]
properties {
 attenuationSettings=&Fort_GadgetActivate_Att
 maxDistance=4000
 soundClassObject=/Game/Sounds/FortSoundClassesAndModes/Fort_Mission_IngameSounds
 volumeMultiplier=1.3
 concurrencySet [&RifleBulletImpact_Conc]
}
